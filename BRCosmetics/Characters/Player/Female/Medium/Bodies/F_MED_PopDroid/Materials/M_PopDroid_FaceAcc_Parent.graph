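# Parent material for the face accessory of the female PopDroid player character, built on the 2023 character parent function stack.
# Material attributes flow from the PreFX function through a PopDroid extra-texture layer, a flipbook face layer and the generic visor function into PostFX.
# The visor function's UV output loops back as the flipbook's custom UVs, and PreFX also supplies its tech art mask.
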Material /BRCosmetics/Characters/Player/Female/Medium/Bodies/F_MED_PopDroid/Materials/M_PopDroid_FaceAcc_Parent
graph {
 properties {
  bEnableNewHLSLGeneratorOverridden=true
  bHasPixelAnimation=true
  bUseMaterialAttributes=true
  bUsedWithMeshDeformer=true
  bUsedWithMorphTargets=true
  bUsedWithSkeletalMesh=true
  pixelDepthOffsetMode=PDOM_Legacy
  shadingModel=MSM_FromMaterialExpression
  usedShadingModels=Default Lit | Subsurface | Clear Coat | Subsurface Profile | Cloth
 }
 nodes [
  n1 = MaterialFunctionCall() {materialFunction=&MF_CharacterParent_2023_PreFX}
  n2 = MaterialFunctionCall!(
   Material=MaterialFunctionCall!(
    CustomUVs=@n2.UVs
    Material=MaterialFunctionCall!(Material=@n1.Result, TechArtMask=@n1.Result) {materialFunction=&/MF_PopDroid_ExtraTexture_FaceAcc}.MaterialAttributes
   ) {materialFunction=&/MF_PopDroid_FlipbookFace}.Material
   Tech Art Mask=@n1.Tech Art Mask
  ) {materialFunction=/Game/Global/Materials/Character_Parent_2023/Visor_Parent_2025/MF_GenericVisor}
  MaterialFunctionCall(Material=@n2.Result, M=@n1.M, Pre Skinned Local Position=@n1.M, Distance Blend=@n1.M, Fresnel=@n1.M) {materialFunction=&MF_CharacterParent_2023_PostFX}
 ]
}
