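# Physics asset for the PopDroid female medium body's jacket, set up for secondary motion of the jacket bones.
# Seven chains (one center, three left, three right) each run from a tiny base capsule constrained to spine_04 through two dynamic capsule segments.
# Dynamic joints use narrow swing limits with locked twist, while base and spine joints rotate freely.
# Large spine_03 and spine_04 capsules supply the torso volume.
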PhysicsAsset /BRCosmetics/Characters/Player/Female/Medium/Bodies/F_MED_PopDroid/Meshes/F_MED_PopDroid_JacketPhysics
bodies {
 base_jacket_c [{shapeType=Capsule, length=0.505, radius=0.505, shapeName=base_jacket_c_capsule, rotation=<rot 0 -90 0>}]
 base_jacket_l_A [{shapeType=Capsule, length=0.505, radius=0.505, shapeName=base_jacket_l_A_capsule, rotation=<rot 0 -90 0>}]
 base_jacket_l_B [{shapeType=Capsule, length=0.505, radius=0.505, shapeName=base_jacket_l_B_capsule, rotation=<rot 0 -90 0>}]
 base_jacket_l_C [{shapeType=Capsule, length=0.505, radius=0.505, shapeName=base_jacket_l_C_capsule, rotation=<rot 0 -90 0>}]
 base_jacket_r_A [{shapeType=Capsule, length=0.505, radius=0.505, shapeName=base_jacket_r_A_capsule, rotation=<rot 0 -90 0>}]
 base_jacket_r_B [{shapeType=Capsule, length=0.505, radius=0.505, shapeName=base_jacket_r_B_capsule, rotation=<rot 0 -90 0>}]
 base_jacket_r_C [{shapeType=Capsule, length=0.505, radius=0.505, shapeName=base_jacket_r_C_capsule, rotation=<rot 0 -90 0>}]
 dyn_jacket_c_01 [{shapeType=Capsule, length=5, radius=2, shapeName=dyn_jacket_c_01_capsule, center=<2 0 0>}]
 dyn_jacket_c_02 [{shapeType=Capsule, length=4, radius=2, shapeName=dyn_jacket_c_02_capsule, center=<4 0 0>, rotation=<rot 90 0 0>}]
 dyn_jacket_l_A_01 [{shapeType=Capsule, length=5, radius=2, shapeName=dyn_jacket_l_A_01_capsule, center=<2 0 0>, rotation=<rot 0 97.1883 0>}]
 dyn_jacket_l_A_02 [{shapeType=Capsule, length=4, radius=2, shapeName=dyn_jacket_l_A_02_capsule, center=<4 0 0>, rotation=<rot 90 0 0>}]
 dyn_jacket_l_B_01 [{shapeType=Capsule, length=5, radius=2, shapeName=dyn_jacket_l_B_01_capsule, center=<2 0 0>, rotation=<rot 0 97.1883 0>}]
 dyn_jacket_l_B_02 [{shapeType=Capsule, length=4, radius=2, shapeName=dyn_jacket_l_B_02_capsule, center=<4 0 0>, rotation=<rot 90 0 0>}]
 dyn_jacket_l_C_01 [{shapeType=Capsule, length=5, radius=2, shapeName=dyn_jacket_l_C_01_capsule, center=<2 0 0>, rotation=<rot 0 97.1883 0>}]
 dyn_jacket_l_C_02 [{shapeType=Capsule, length=4, radius=2, shapeName=dyn_jacket_l_C_02_capsule, center=<4 0 0>, rotation=<rot 90 0 0>}]
 dyn_jacket_r_A_01 [{shapeType=Capsule, length=5, radius=2, shapeName=dyn_jacket_r_A_01_capsule, center=<2 0 0>, rotation=<rot 0 97.1882 0>}]
 dyn_jacket_r_A_02 [{shapeType=Capsule, length=4, radius=2, shapeName=dyn_jacket_r_A_02_capsule, center=<4 0 0>, rotation=<rot 90 0 0>}]
 dyn_jacket_r_B_01 [{shapeType=Capsule, length=5, radius=2, shapeName=dyn_jacket_r_B_01_capsule, center=<2 0 0>, rotation=<rot 0 97.1883 0>}]
 dyn_jacket_r_B_02 [{shapeType=Capsule, length=4, radius=2, shapeName=dyn_jacket_r_B_02_capsule, center=<4 0 0>, rotation=<rot 90 0 0>}]
 dyn_jacket_r_C_01 [{shapeType=Capsule, length=5, radius=2, shapeName=dyn_jacket_r_C_01_capsule, center=<2 0 0>, rotation=<rot 0 97.1883 0>}]
 dyn_jacket_r_C_02 [{shapeType=Capsule, length=4, radius=2, shapeName=dyn_jacket_r_C_02_capsule, center=<4 0 0>, rotation=<rot 90 0 0>}]
 spine_03 [{shapeType=Capsule, length=3.8366, radius=9.247, shapeName=spine_03_capsule, center=<1.1739 2.0346 0>}]
 spine_04 [{shapeType=Capsule, length=4.1735, radius=9.6822, shapeName=spine_04_capsule, center=<0.2455 1.7543 0>}]
}
constraints [
 {
  bone1Name=spine_04
  bone2Name=spine_03
  swing1LimitDegrees=45
  swing1Motion=Free
  swing2LimitDegrees=45
  swing2Motion=Free
  twistLimitDegrees=45
  twistMotion=Free
 }
 {
  bone1Name=base_jacket_c
  bone2Name=spine_04
  swing1LimitDegrees=45
  swing1Motion=Free
  swing2LimitDegrees=45
  swing2Motion=Free
  twistLimitDegrees=45
  twistMotion=Free
 }
 {bone1Name=dyn_jacket_c_01, bone2Name=base_jacket_c, swing1LimitDegrees=15, swing2LimitDegrees=10, twistLimitDegrees=45, twistMotion=Locked}
 {bone1Name=dyn_jacket_c_02, bone2Name=dyn_jacket_c_01, swing1LimitDegrees=15, swing2LimitDegrees=10, twistLimitDegrees=45, twistMotion=Locked}
 {
  bone1Name=base_jacket_l_A
  bone2Name=spine_04
  swing1LimitDegrees=45
  swing1Motion=Free
  swing2LimitDegrees=45
  swing2Motion=Free
  twistLimitDegrees=45
  twistMotion=Free
 }
 {
  bone1Name=dyn_jacket_l_A_01
  bone2Name=base_jacket_l_A
  swing1LimitDegrees=15
  swing2LimitDegrees=10
  twistLimitDegrees=45
  twistMotion=Locked
 }
 {
  bone1Name=dyn_jacket_l_A_02
  bone2Name=dyn_jacket_l_A_01
  swing1LimitDegrees=15
  swing2LimitDegrees=10
  twistLimitDegrees=45
  twistMotion=Locked
 }
 {
  bone1Name=base_jacket_l_B
  bone2Name=spine_04
  swing1LimitDegrees=45
  swing1Motion=Free
  swing2LimitDegrees=45
  swing2Motion=Free
  twistLimitDegrees=45
  twistMotion=Free
 }
 {
  bone1Name=dyn_jacket_l_B_01
  bone2Name=base_jacket_l_B
  swing1LimitDegrees=15
  swing2LimitDegrees=10
  twistLimitDegrees=45
  twistMotion=Locked
 }
 {
  bone1Name=dyn_jacket_l_B_02
  bone2Name=dyn_jacket_l_B_01
  swing1LimitDegrees=15
  swing2LimitDegrees=10
  twistLimitDegrees=45
  twistMotion=Locked
 }
 {
  bone1Name=base_jacket_l_C
  bone2Name=spine_04
  swing1LimitDegrees=45
  swing1Motion=Free
  swing2LimitDegrees=45
  swing2Motion=Free
  twistLimitDegrees=45
  twistMotion=Free
 }
 {
  bone1Name=dyn_jacket_l_C_01
  bone2Name=base_jacket_l_C
  swing1LimitDegrees=15
  swing2LimitDegrees=10
  twistLimitDegrees=45
  twistMotion=Locked
 }
 {
  bone1Name=dyn_jacket_l_C_02
  bone2Name=dyn_jacket_l_C_01
  swing1LimitDegrees=15
  swing2LimitDegrees=10
  twistLimitDegrees=45
  twistMotion=Locked
 }
 {
  bone1Name=base_jacket_r_A
  bone2Name=spine_04
  swing1LimitDegrees=45
  swing1Motion=Free
  swing2LimitDegrees=45
  swing2Motion=Free
  twistLimitDegrees=45
  twistMotion=Free
 }
 {
  bone1Name=dyn_jacket_r_A_01
  bone2Name=base_jacket_r_A
  swing1LimitDegrees=15
  swing2LimitDegrees=10
  twistLimitDegrees=45
  twistMotion=Locked
 }
 {
  bone1Name=dyn_jacket_r_A_02
  bone2Name=dyn_jacket_r_A_01
  swing1LimitDegrees=15
  swing2LimitDegrees=10
  twistLimitDegrees=45
  twistMotion=Locked
 }
 {
  bone1Name=base_jacket_r_B
  bone2Name=spine_04
  swing1LimitDegrees=45
  swing1Motion=Free
  swing2LimitDegrees=45
  swing2Motion=Free
  twistLimitDegrees=45
  twistMotion=Free
 }
 {
  bone1Name=dyn_jacket_r_B_01
  bone2Name=base_jacket_r_B
  swing1LimitDegrees=15
  swing2LimitDegrees=10
  twistLimitDegrees=45
  twistMotion=Locked
 }
 {
  bone1Name=dyn_jacket_r_B_02
  bone2Name=dyn_jacket_r_B_01
  swing1LimitDegrees=15
  swing2LimitDegrees=10
  twistLimitDegrees=45
  twistMotion=Locked
 }
 {
  bone1Name=base_jacket_r_C
  bone2Name=spine_04
  swing1LimitDegrees=45
  swing1Motion=Free
  swing2LimitDegrees=45
  swing2Motion=Free
  twistLimitDegrees=45
  twistMotion=Free
 }
 {
  bone1Name=dyn_jacket_r_C_01
  bone2Name=base_jacket_r_C
  swing1LimitDegrees=15
  swing2LimitDegrees=10
  twistLimitDegrees=45
  twistMotion=Locked
 }
 {
  bone1Name=dyn_jacket_r_C_02
  bone2Name=dyn_jacket_r_C_01
  swing1LimitDegrees=15
  swing2LimitDegrees=10
  twistLimitDegrees=45
  twistMotion=Locked
 }
]
