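# Physics asset for the dangling straps on the F_MED_PopDroid female player body cosmetic, defining their simulated shapes and joints.
# A two-link left chain and a four-link right chain of capsules hang from base strap bones linked to spine_04, with angle-limited joints and locked twist at each chain's first link.
# Capsules on the pelvis, spine, thighs and right forearm provide body shapes around the straps, joined by free-rotation constraints.
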PhysicsAsset /BRCosmetics/Characters/Player/Female/Medium/Bodies/F_MED_PopDroid/Meshes/F_MED_PopDroid_StrapsPhysics
bodies {
 base_strap_l [{shapeType=Capsule, length=1, radius=1, shapeName=base_strap_l_capsule}]
 base_strap_r [{shapeType=Capsule, length=1, radius=1, shapeName=base_strap_r_capsule}]
 dyn_strap_l_01 [{shapeType=Capsule, length=1, radius=1, shapeName=dyn_strap_l_01_capsule, center=<2 0 0>, rotation=<rot 90 0 0>}]
 dyn_strap_l_02 [{shapeType=Capsule, length=1, radius=1, shapeName=dyn_strap_l_02_capsule, center=<2 0 0>, rotation=<rot 90 0 0>}]
 dyn_strap_r_01 [{shapeType=Capsule, length=2, radius=1.2, shapeName=dyn_strap_r_01_capsule, center=<2 0 0>, rotation=<rot 90 0 0>}]
 dyn_strap_r_02 [{shapeType=Capsule, length=2, radius=1.2, shapeName=dyn_strap_r_02_capsule, center=<2 0 0>, rotation=<rot 90 0 0>}]
 dyn_strap_r_03 [{shapeType=Capsule, length=2, radius=1.2, shapeName=dyn_strap_r_03_capsule, center=<2 0 0>, rotation=<rot 90 0 0>}]
 dyn_strap_r_04 [{shapeType=Capsule, length=2.3564, radius=1.6436, shapeName=dyn_strap_r_04_capsule, center=<1.196 0 0>}]
 lowerarm_r [{shapeType=Capsule, length=16.274, radius=4.8555, shapeName=lowerarm_r_capsule, center=<-10.4588 0 0>, rotation=<rot 90 0 0>}]
 pelvis [{shapeType=Capsule, length=12.2561, radius=11.2547, shapeName=pelvis_capsule, center=<0.5333 -2.0614 0>, rotation=<rot 0 -90 0>}]
 spine_01 [{shapeType=Capsule, length=5.3493, radius=8.4343, shapeName=spine_01_capsule, center=<4.1605 1.3071 0>, rotation=<rot 0 -90 0>}]
 spine_02 [{
   shapeType=Capsule
   length=3.9506
   radius=8.1301
   shapeName=spine_02_capsule
   center=<3.5798 2.5169 0.3474>
   rotation=<rot -7.8584 -90 1.2675>
  }]
 spine_03 [{shapeType=Capsule, length=0.505, radius=0.505, shapeName=spine_03_capsule, rotation=<rot 0 -90 0>}]
 spine_04 [{
   shapeType=Capsule
   length=6.1075
   radius=8.7093
   shapeName=spine_04_capsule
   center=<-1.3609 3.6445 1.1471>
   rotation=<rot -17.4711 -90 5.1853>
  }]
 thigh_l [{
   shapeType=Capsule
   length=22.4397
   radius=9.314
   shapeName=thigh_l_capsule
   center=<-20.9223 0.4117 -0.1478>
   rotation=<rot 0.2467 -86.6252 -88.7914>
  }]
 thigh_r [{
   shapeType=Capsule
   length=22.4397
   radius=9.314
   shapeName=thigh_r_capsule
   center=<20.9223 -0.4117 0.1478>
   rotation=<rot 0.2467 -86.6252 91.2086>
  }]
}
constraints [
 {bone1Name=dyn_strap_l_01, bone2Name=base_strap_l, swing1LimitDegrees=45, swing2LimitDegrees=30, twistLimitDegrees=15, twistMotion=Locked}
 {bone1Name=dyn_strap_l_02, bone2Name=dyn_strap_l_01, swing1LimitDegrees=45, swing2LimitDegrees=30, twistLimitDegrees=15}
 {bone1Name=dyn_strap_r_01, bone2Name=base_strap_r, swing1LimitDegrees=45, swing2LimitDegrees=30, twistLimitDegrees=15, twistMotion=Locked}
 {bone1Name=dyn_strap_r_02, bone2Name=dyn_strap_r_01, swing1LimitDegrees=45, swing2LimitDegrees=30, twistLimitDegrees=15}
 {bone1Name=dyn_strap_r_03, bone2Name=dyn_strap_r_02, swing1LimitDegrees=45, swing2LimitDegrees=30, twistLimitDegrees=15}
 {bone1Name=dyn_strap_r_04, bone2Name=dyn_strap_r_03, swing1LimitDegrees=45, swing2LimitDegrees=30, twistLimitDegrees=15}
 {
  bone1Name=base_strap_r
  bone2Name=spine_04
  swing1LimitDegrees=45
  swing1Motion=Free
  swing2LimitDegrees=45
  swing2Motion=Free
  twistLimitDegrees=45
  twistMotion=Free
 }
 {
  bone1Name=base_strap_l
  bone2Name=spine_04
  swing1LimitDegrees=45
  swing1Motion=Free
  swing2LimitDegrees=45
  swing2Motion=Free
  twistLimitDegrees=45
  twistMotion=Free
 }
 {
  bone1Name=spine_04
  bone2Name=spine_03
  swing1LimitDegrees=45
  swing1Motion=Free
  swing2LimitDegrees=45
  swing2Motion=Free
  twistLimitDegrees=45
  twistMotion=Free
 }
 {
  bone1Name=spine_03
  bone2Name=spine_02
  swing1LimitDegrees=45
  swing1Motion=Free
  swing2LimitDegrees=45
  swing2Motion=Free
  twistLimitDegrees=45
  twistMotion=Free
 }
 {
  bone1Name=spine_02
  bone2Name=spine_01
  swing1LimitDegrees=45
  swing1Motion=Free
  swing2LimitDegrees=45
  swing2Motion=Free
  twistLimitDegrees=45
  twistMotion=Free
 }
 {
  bone1Name=spine_01
  bone2Name=pelvis
  swing1LimitDegrees=45
  swing1Motion=Free
  swing2LimitDegrees=45
  swing2Motion=Free
  twistLimitDegrees=45
  twistMotion=Free
 }
 {
  bone1Name=thigh_l
  bone2Name=pelvis
  swing1LimitDegrees=45
  swing1Motion=Free
  swing2LimitDegrees=45
  swing2Motion=Free
  twistLimitDegrees=45
  twistMotion=Free
 }
 {
  bone1Name=thigh_r
  bone2Name=pelvis
  swing1LimitDegrees=45
  swing1Motion=Free
  swing2LimitDegrees=45
  swing2Motion=Free
  twistLimitDegrees=45
  twistMotion=Free
 }
 {
  bone1Name=lowerarm_r
  bone2Name=spine_04
  swing1LimitDegrees=45
  swing1Motion=Free
  swing2LimitDegrees=45
  swing2Motion=Free
  twistLimitDegrees=45
  twistMotion=Free
 }
]
