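# Post-process style AnimBP for the female PopDroid cosmetic body, adding secondary motion to the pose copied from the parent mesh.
# A Control Rig receives crouch, slide, skydive-dive and gun-out states.
# Rigid-body jacket simulation feeds weighted strap constraints and a strap simulation, then additive jacket/collar rotations and pelvis-relative Anim Dynamics for belt and gun bones.
# The EventGraph sets belt and jacket alphas from movement and zipline events, and gun-out from weapon type.
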AnimBlueprint /BRCosmetics/Characters/Player/Female/Medium/Bodies/F_MED_PopDroid/Meshes/F_MED_PopDroid_AnimBP
target_skeleton=&/F_MED_PopDroid_Skeleton
graphs [
 AnimGraph {nodes [
   Output Pose(Result=Use cached pose 'outputCache'()) {layerGroup=None}
   n1 = Control Rig!(
    Source=Copy Pose From Mesh() {bUseAttachedParent=true}
    Alpha
    Crouching:bool=Get bIsCrouching()
    sliding:bool=Get bIsSliding()
    SkydiveDiveMode:bool=Get bIsSkydiveDiveMode()
    isWeaponEquipped:double=Get isGunOut() -> (isGunOut:double)
   ) -> (Pose) {
    alphaBoolBlend {blendOption=Linear}
    controlRigAssetReference {blueprintRigClass=&/F_MED_PopDroid_CtrlRig.F_MED_PopDroid_CtrlRig_C}
   }
   Save cached pose 'inputCache'(Pose=@n1) {cacheName=inputCache}
   n2 = RigidBody(ComponentPose=Local To Component(LocalPose=Use cached pose 'inputCache'()), Alpha=0.8) {
    bClampLinearTranslationLimitToRefPose=true
    overridePhysicsAsset=&/F_MED_PopDroid_JacketPhysics
    simulationSpace=BaseBoneSpace
    alphaBoolBlend {blendOption=Linear}
    alphaScaleBiasClamp {bInterpResult=true}
    baseBoneRef {boneName=pelvis}
    simSpaceSettings {
     maxAngularAcceleration=1
     maxAngularVelocity=3
     maxLinearAcceleration=1000
     maxLinearVelocity=500
     velocityScaleZ=0
     worldAlpha=1
     externalLinearDragV=<1 1 1>
    }
   }
   n3 = "Constraint - base_strap_r (PPPPPPPP)"!(
    ComponentPose=@n2
    Alpha=1
    ConstraintWeights_0=0.7708
    ConstraintWeights_1:float=0.1032
    ConstraintWeights_2:float=0.0913
    ConstraintWeights_3:float=0.0214
    ConstraintWeights_4:float=0.0069
    ConstraintWeights_5:float=0.0043
    ConstraintWeights_6:float=0.0016
    ConstraintWeights_7:float=0.0006
   ) -> (Pose) {
    alphaBoolBlend {blendOption=Linear}
    boneToModify {boneName=base_strap_r}
    constraintSetup [
     {targetBone {boneName=dyn_jacket_r_A_02}}
     {targetBone {boneName=dyn_jacket_r_A_01}}
     {targetBone {boneName=dyn_jacket_r_B_02}}
     {targetBone {boneName=dyn_jacket_r_B_01}}
     {targetBone {boneName=dyn_jacket_r_C_02}}
     {targetBone {boneName=spine_04}}
     {targetBone {boneName=dyn_jacket_r_C_01}}
     {targetBone {boneName=spine_02}}
    ]
    constraintWeights [1,1,1,1,1,1,1,1]
   }
   n4 = "Constraint - base_strap_l (PPPPPPPP)"!(
    ComponentPose=@n3
    Alpha=1
    ConstraintWeights_0=0.6916
    ConstraintWeights_1:float=0.1418
    ConstraintWeights_2:float=0.1059
    ConstraintWeights_3:float=0.0327
    ConstraintWeights_4:float=0.0128
    ConstraintWeights_5:float=0.0111
    ConstraintWeights_6:float=0.003
    ConstraintWeights_7:float=0.0012
   ) -> (Pose) {
    alphaBoolBlend {blendOption=Linear}
    boneToModify {boneName=base_strap_l}
    constraintSetup [
     {targetBone {boneName=dyn_jacket_l_A_02}}
     {targetBone {boneName=dyn_jacket_l_A_01}}
     {targetBone {boneName=dyn_jacket_l_B_02}}
     {targetBone {boneName=dyn_jacket_l_B_01}}
     {targetBone {boneName=dyn_jacket_l_C_02}}
     {targetBone {boneName=spine_04}}
     {targetBone {boneName=dyn_jacket_l_C_01}}
     {targetBone {boneName=spine_02}}
    ]
    constraintWeights [1,1,1,1,1,1,1,1]
   }
   n5 = RigidBody(ComponentPose=@n4) {
    bClampLinearTranslationLimitToRefPose=true
    overridePhysicsAsset=&/F_MED_PopDroid_StrapsPhysics
    simulationSpace=BaseBoneSpace
    alphaBoolBlend {blendOption=Linear}
    baseBoneRef {boneName=pelvis}
    simSpaceSettings {
     maxAngularAcceleration=1
     maxAngularVelocity=3
     maxLinearAcceleration=1000
     maxLinearVelocity=500
     velocityScaleZ=0
     worldAlpha=1
     externalLinearDragV=<1 1 1>
    }
   }
   n6 = "Transform (Modify) Bone - Bone: base_jacket_c"!(ComponentPose=@n5, bAlphaBoolEnabled:bool=True, Rotation="0,-10,0") -> (Pose) {
    alphaInputType=Bool
    rotationMode=BMM_Additive
    alphaBoolBlend {blendInTime=0.2, blendOption=Cubic, blendOutTime=0.2}
    boneToModify {boneName=base_jacket_c}
    rotation {yaw=-10}
   }
   n7 = "Transform (Modify) Bone - Bone: base_jacket_l_C"!(ComponentPose=@n6, bAlphaBoolEnabled:bool=True, Rotation="0,-10,0") -> (Pose) {
    alphaInputType=Bool
    rotationMode=BMM_Additive
    alphaBoolBlend {blendInTime=0.2, blendOption=Cubic, blendOutTime=0.2}
    boneToModify {boneName=base_jacket_l_C}
    rotation {yaw=-10}
   }
   n8 = "Transform (Modify) Bone - Bone: base_collar_r_C"!(ComponentPose=@n7, bAlphaBoolEnabled:bool=True, Rotation="0,-10,0") -> (Pose) {
    alphaInputType=Bool
    rotationMode=BMM_Additive
    alphaBoolBlend {blendInTime=0.2, blendOption=Cubic, blendOutTime=0.2}
    boneToModify {boneName=base_collar_r_C}
    rotation {yaw=-10}
   }
   n9 = "Transform (Modify) Bone - Bone: base_jacket_l_B"!(ComponentPose=@n8, bAlphaBoolEnabled:bool=True, Rotation="0,-10,0") -> (Pose) {
    alphaInputType=Bool
    rotationMode=BMM_Additive
    alphaBoolBlend {blendInTime=0.2, blendOption=Cubic, blendOutTime=0.2}
    boneToModify {boneName=base_jacket_l_B}
    rotation {yaw=-10}
   }
   n10 = "Transform (Modify) Bone - Bone: base_collar_r_B"!(ComponentPose=@n9, bAlphaBoolEnabled:bool=True, Rotation="0,-10,0") -> (Pose) {
    alphaInputType=Bool
    rotationMode=BMM_Additive
    alphaBoolBlend {blendInTime=0.2, blendOption=Cubic, blendOutTime=0.2}
    boneToModify {boneName=base_collar_r_B}
    rotation {yaw=-10}
   }
   n11 = Anim Dynamics - Bone: dyn_belt_02!(ComponentPose=@n10, Alpha=1, GravityOverride:Vector="0,0,-980") -> (Pose) {
    angularDampingOverride=1
    angularSpringConstant=20000
    bOverrideAngularDamping=true
    bUseGravityOverride=true
    gravityScale=0.35
    alphaBoolBlend {blendOption=Linear}
    alphaScaleBiasClamp {bInterpResult=true}
    boundBone {boneName=dyn_belt_02}
    componentLinearAccScale=<1 1 1>
    componentLinearVelScale=<1 1 1>
    gravityOverride {z=-980}
    physicsBodyDefinitions [{
      boundBone {boneName=dyn_belt_02}
      boxExtents=<28 28 28>
      constraintSetup {angularLimitsMax=<0 0 0>, angularLimitsMin=<0 0 -10>, angularTarget=<1 0 0>}
      localJointOffset=<2 0 0>
     }]
    relativeSpaceBone {boneName=pelvis}
   }
   n12 = Anim Dynamics - Bone: dyn_belt_03!(ComponentPose=@n11, Alpha=1, GravityOverride:Vector="0,0,-980") -> (Pose) {
    angularDampingOverride=1
    angularSpringConstant=20000
    bOverrideAngularDamping=true
    bUseGravityOverride=true
    gravityScale=0.35
    alphaBoolBlend {blendOption=Linear}
    alphaScaleBiasClamp {bInterpResult=true}
    boundBone {boneName=dyn_belt_03}
    componentLinearAccScale=<1 1 1>
    componentLinearVelScale=<1 1 1>
    gravityOverride {z=-980}
    physicsBodyDefinitions [{
      boundBone {boneName=dyn_belt_03}
      boxExtents=<28 28 28>
      constraintSetup {angularLimitsMax=<0 0 0>, angularLimitsMin=<0 0 -10>, angularTarget=<1 0 0>}
      localJointOffset=<2 0 0>
     }]
    relativeSpaceBone {boneName=pelvis}
   }
   n13 = Anim Dynamics - Bone: dyn_belt_04!(ComponentPose=@n12, Alpha=1, GravityOverride:Vector="0,0,-980") -> (Pose) {
    angularDampingOverride=1
    angularSpringConstant=20000
    bOverrideAngularDamping=true
    bUseGravityOverride=true
    gravityScale=0.35
    alphaBoolBlend {blendOption=Linear}
    alphaScaleBiasClamp {bInterpResult=true}
    boundBone {boneName=dyn_belt_04}
    componentLinearAccScale=<1 1 1>
    componentLinearVelScale=<1 1 1>
    gravityOverride {z=-980}
    physicsBodyDefinitions [{
      boundBone {boneName=dyn_belt_04}
      boxExtents=<28 28 28>
      constraintSetup {angularLimitsMax=<0 0 0>, angularLimitsMin=<0 0 -10>, angularTarget=<1 0 0>}
      localJointOffset=<2 0 0>
     }]
    relativeSpaceBone {boneName=pelvis}
   }
   n14 = Anim Dynamics - Bone: dyn_belt_05!(ComponentPose=@n13, Alpha=1, GravityOverride:Vector="0,0,-980") -> (Pose) {
    angularDampingOverride=1
    angularSpringConstant=20000
    bOverrideAngularDamping=true
    bUseGravityOverride=true
    gravityScale=0.35
    alphaBoolBlend {blendOption=Linear}
    alphaScaleBiasClamp {bInterpResult=true}
    boundBone {boneName=dyn_belt_05}
    componentLinearAccScale=<1 1 1>
    componentLinearVelScale=<1 1 1>
    gravityOverride {z=-980}
    physicsBodyDefinitions [{
      boundBone {boneName=dyn_belt_05}
      boxExtents=<28 28 28>
      constraintSetup {angularLimitsMax=<5 5 0>, angularLimitsMin=<-5 -5 -10>, angularTarget=<1 0 0>}
      localJointOffset=<2 0 0>
     }]
    relativeSpaceBone {boneName=pelvis}
   }
   n15 = Anim Dynamics - Bone: dyn_gun!(ComponentPose=@n14, Alpha=1, GravityOverride:Vector="0,0,-980") -> (Pose) {
    angularDampingOverride=1
    angularSpringConstant=20000
    bOverrideAngularDamping=true
    bUseGravityOverride=true
    gravityScale=0.35
    alphaBoolBlend {blendOption=Linear}
    alphaScaleBiasClamp {bInterpResult=true}
    boundBone {boneName=dyn_gun}
    componentLinearAccScale=<1 1 1>
    componentLinearVelScale=<1 1 1>
    gravityOverride {z=-980}
    physicsBodyDefinitions [{
      boundBone {boneName=dyn_gun}
      boxExtents=<28 28 28>
      constraintSetup {angularLimitsMax=<0 0 0>, angularLimitsMin=<0 0 -10>, angularTarget=<1 0 0>}
      localJointOffset=<2 0 0>
     }]
    relativeSpaceBone {boneName=pelvis}
   }
   Save cached pose 'outputCache'(Pose=Component To Local(ComponentPose=@n15)) {cacheName=outputCache}
  ]}
 EventGraph {nodes [
   Event BlueprintUpdateAnimation()
   n1 = SelectFloat(A=0.25, B=1, bPickA=Get bIsSwimSprinting() -> (bIsSwimSprinting:bool))
   > Set beltAlpha(beltAlpha:double=@n1) -> (Output_Get:double)
   > Set BeltAlphaReduc(BeltAlphaReduc:double=float * float(A=@n1, B=0.25) {is=title K2Node_PromotableOperator}) -> (Output_Get:double)
   > Set JacketAlpha(JacketAlpha:double=SelectFloat(A=0.5, B=0.8, bPickA=Get bIsSkyDiving())) -> (Output_Get:double)
   n2 = Event BlueprintProcessAnimInputEvent()
   > n3 = Branch(Condition=IsNamed(self=@n2.AnimInputEvent, InEventName=aie.Player.Zipline.Entered))
   > Set beltAlpha(beltAlpha=0)
   Branch(execute=@n3.else, Condition=IsNamed(self=@n2.AnimInputEvent, InEventName=aie.Player.Zipline.Exited))
   > Set beltAlpha(beltAlpha=1)
   Event OnWeaponChanged()
   > Set isGunOut(isGunOut:double=1) -> (Output_Get:double)
   > n4 = Switch on EFortWeaponCoreAnimation!(execute, Selection:EFortWeaponCoreAnimation Enum=Get WeaponCoreAnim!() -> (WeaponCoreAnim)) -> (
    Melee:Exec
    Pistol:Exec
    Shotgun:Exec
    PaperBlueprint:Exec
    Rifle:Exec
    MeleeOneHand:Exec
    MachinePistol:Exec
    RocketLauncher:Exec
    GrenadeLauncher:Exec
    GoingCommando:Exec
    AssaultRifle:Exec
    TacticalShotgun:Exec
    SniperRifle:Exec
    TrapPlacement:Exec
    ShoulderLauncher:Exec
    AbilityDecoTool:Exec
    Crossbow:Exec
    C4:Exec
    RemoteControl:Exec
    DualWield:Exec
    AR_BullPup:Exec
    AR_ForwardGrip:Exec
    MedPackPaddles:Exec
    SMG_P90:Exec
    AR_DrumGun:Exec
    Consumable_Small:Exec
    Consumable_Large:Exec
    Balloon:Exec
    MountedTurret:Exec
    CreativeTool:Exec
    ExplosiveBow:Exec
    AshtonIndigo:Exec
    AshtonChicago:Exec
    MeleeDualWield:Exec
    Unarmed:Exec
   )
   Set isGunOut(execute=[@n4.Melee, @n4.MeleeDualWield, @n4.MeleeOneHand, @n4.Unarmed], isGunOut=0)
  ]}
]
variables [
 beltAlpha {type=number, default=1}
 BeltAlphaReduc {type=number, default=0.25}
 beltConstraint {type=number, default=1}
 isGunOut {type=number}
 JacketAlpha {type=number, default=0.8}
]
